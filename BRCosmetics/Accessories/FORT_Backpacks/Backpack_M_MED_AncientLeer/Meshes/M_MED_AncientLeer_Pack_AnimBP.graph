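# Animation Blueprint for the AncientLeer backpack cosmetic, posing the pack body and simulating its seven leaf bones.
# Additive bone transforms offset the body, with extra front-end and skydiving adjustments, plus one leaf, into a cached base pose.
# That pose feeds two world-space AnimDynamics spring chains, blended by bIsInFrontEnd: heavily damped for front end, livelier otherwise.
# The update event derives Parashuting from spine yaw, but the shown graphs never read it.
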AnimBlueprint /BRCosmetics/Accessories/FORT_Backpacks/Backpack_M_MED_AncientLeer/Meshes/M_MED_AncientLeer_Pack_AnimBP
target_skeleton=&/M_MED_AncientLeer_Pack_Skeleton
graphs [
 AnimGraph {nodes [
   n1 = Anim Dynamics - Bone: Dyn_leaves_01(ComponentPose=Local To Component(LocalPose=Use cached pose 'Base'()), Alpha=1, ExternalForce="0,0,0") {
    angularDampingOverride=0.9
    angularSpringConstant=800
    bAngularSpring=true
    bGravityOverrideInSimSpace=true
    bOverrideAngularDamping=true
    bUseGravityOverride=true
    simulationSpace=World
    alphaBoolBlend {blendOption=Linear}
    boundBone {boneName=Dyn_leaves_01}
    componentAppliedLinearAccClamp=<700 700 700>
    componentLinearAccScale=<0.65 0.64 1>
    componentLinearVelScale=<1 1 1>
    gravityOverride {z=-500}
    physicsBodyDefinitions [{
      boundBone {boneName=Dyn_leaves_01}
      boxExtents=<10 10 10>
      constraintSetup {angularLimitsMax=<10 25 20>, angularLimitsMin=<-10 -25 -20>, angularTarget=<1 0 0>}
      localJointOffset=<1 0 0>
     }]
   }
   n2 = Anim Dynamics - Bone: Dyn_leaves_02(ComponentPose=@n1, Alpha=1, ExternalForce="0,0,0") {
    angularDampingOverride=0.8
    angularSpringConstant=750
    bAngularSpring=true
    bGravityOverrideInSimSpace=true
    bOverrideAngularDamping=true
    bUseGravityOverride=true
    simulationSpace=World
    alphaBoolBlend {blendOption=Linear}
    boundBone {boneName=Dyn_leaves_02}
    componentAppliedLinearAccClamp=<500 500 500>
    componentLinearAccScale=<0.5 0.5 0.8>
    componentLinearVelScale=<0.9 0.9 0.9>
    gravityOverride {z=-500}
    physicsBodyDefinitions [{
      boundBone {boneName=Dyn_leaves_02}
      boxExtents=<10 10 10>
      constraintSetup {angularLimitsMax=<10 35 20>, angularLimitsMin=<-10 -35 -20>, angularTarget=<1 0 0>}
      localJointOffset=<1 0 0>
     }]
   }
   n3 = Anim Dynamics - Bone: Dyn_leaves_03(ComponentPose=@n2, Alpha=1, ExternalForce="0,0,0") {
    angularDampingOverride=0.875
    angularSpringConstant=700
    bAngularSpring=true
    bGravityOverrideInSimSpace=true
    bOverrideAngularDamping=true
    bUseGravityOverride=true
    simulationSpace=World
    alphaBoolBlend {blendOption=Linear}
    boundBone {boneName=Dyn_leaves_03}
    componentAppliedLinearAccClamp=<600 600 600>
    componentLinearAccScale=<0.6 0.6 1>
    componentLinearVelScale=<1 1 1>
    gravityOverride {z=-500}
    physicsBodyDefinitions [{
      boundBone {boneName=Dyn_leaves_03}
      boxExtents=<10 10 10>
      constraintSetup {angularLimitsMax=<10 30 20>, angularLimitsMin=<-10 -30 -20>, angularTarget=<1 0 0>}
      localJointOffset=<1 0 0>
     }]
   }
   n4 = Anim Dynamics - Bone: Dyn_leaves_04(ComponentPose=@n3, Alpha=1, ExternalForce="0,0,0") {
    angularDampingOverride=0.75
    angularSpringConstant=800
    bAngularSpring=true
    bGravityOverrideInSimSpace=true
    bOverrideAngularDamping=true
    bUseGravityOverride=true
    simulationSpace=World
    alphaBoolBlend {blendOption=Linear}
    boundBone {boneName=Dyn_leaves_04}
    componentAppliedLinearAccClamp=<500 500 500>
    componentLinearAccScale=<0.5 0.5 0.9>
    componentLinearVelScale=<0.9 0.9 0.9>
    gravityOverride {z=-500}
    physicsBodyDefinitions [{
      boundBone {boneName=Dyn_leaves_04}
      boxExtents=<10 10 10>
      constraintSetup {angularLimitsMax=<10 25 20>, angularLimitsMin=<-5 -25 -15>, angularTarget=<1 0 0>}
      localJointOffset=<0.7 0 0>
     }]
   }
   n5 = Anim Dynamics - Bone: Dyn_leaves_05(ComponentPose=@n4, Alpha=1, ExternalForce="0,0,0") {
    angularDampingOverride=0.92
    angularSpringConstant=750
    bAngularSpring=true
    bGravityOverrideInSimSpace=true
    bOverrideAngularDamping=true
    bUseGravityOverride=true
    simulationSpace=World
    alphaBoolBlend {blendOption=Linear}
    boundBone {boneName=Dyn_leaves_05}
    componentAppliedLinearAccClamp=<500 500 500>
    componentLinearAccScale=<0.5 0.5 0.9>
    componentLinearVelScale=<0.9 0.9 0.9>
    gravityOverride {z=-500}
    physicsBodyDefinitions [{
      boundBone {boneName=Dyn_leaves_05}
      boxExtents=<10 10 10>
      constraintSetup {angularLimitsMax=<10 35 20>, angularLimitsMin=<-5 -35 -15>, angularTarget=<1 0 0>}
      localJointOffset=<0.7 0 0>
     }]
   }
   n6 = Anim Dynamics - Bone: Dyn_leaves_06(ComponentPose=@n5, Alpha=1, ExternalForce="0,0,0") {
    angularDampingOverride=0.87
    angularSpringConstant=700
    bAngularSpring=true
    bGravityOverrideInSimSpace=true
    bOverrideAngularDamping=true
    bUseGravityOverride=true
    simulationSpace=World
    alphaBoolBlend {blendOption=Linear}
    boundBone {boneName=Dyn_leaves_06}
    componentAppliedLinearAccClamp=<500 500 500>
    componentLinearAccScale=<0.5 0.5 0.9>
    componentLinearVelScale=<0.9 0.9 0.9>
    gravityOverride {z=-500}
    physicsBodyDefinitions [{
      boundBone {boneName=Dyn_leaves_06}
      boxExtents=<10 10 10>
      constraintSetup {angularLimitsMax=<10 30 20>, angularLimitsMin=<-5 -30 -15>, angularTarget=<1 0 0>}
      localJointOffset=<0.7 0 0>
     }]
   }
   n7 = Anim Dynamics - Bone: Dyn_leaves_07(ComponentPose=@n6, Alpha=1, ExternalForce="0,0,0") {
    angularDampingOverride=0.8
    angularSpringConstant=800
    bAngularSpring=true
    bGravityOverrideInSimSpace=true
    bOverrideAngularDamping=true
    bUseGravityOverride=true
    simulationSpace=World
    alphaBoolBlend {blendOption=Linear}
    boundBone {boneName=Dyn_leaves_07}
    componentAppliedLinearAccClamp=<700 700 700>
    componentLinearAccScale=<0.65 0.64 1>
    componentLinearVelScale=<1 1 1>
    gravityOverride {z=-500}
    physicsBodyDefinitions [{
      boundBone {boneName=Dyn_leaves_07}
      boxExtents=<10 10 10>
      constraintSetup {angularLimitsMax=<10 32 20>, angularLimitsMin=<-10 -32 0>, angularTarget=<1 0 0>}
      localJointOffset=<1 0 0>
     }]
   }
   n8 = Anim Dynamics - Bone: Dyn_leaves_01(ComponentPose=Local To Component(LocalPose=Use cached pose 'Base'()), Alpha=1, ExternalForce="0,0,0") {
    angularDampingOverride=0.999
    angularSpringConstant=400
    bAngularSpring=true
    bOverrideAngularDamping=true
    bUseGravityOverride=true
    simulationSpace=World
    alphaBoolBlend {blendOption=Linear}
    boundBone {boneName=Dyn_leaves_01}
    componentAppliedLinearAccClamp=<500 500 500>
    componentLinearAccScale=<0.2 0.2 0.2>
    componentLinearVelScale=<0.2 0.2 0.2>
    gravityOverride {z=-980}
    physicsBodyDefinitions [{
      boundBone {boneName=Dyn_leaves_01}
      boxExtents=<10 10 10>
      constraintSetup {angularLimitsMax=<0 13 13>, angularLimitsMin=<0 -30 -13>, angularTarget=<1 0 0>}
      localJointOffset=<0.1 0 0>
     }]
   }
   n9 = Anim Dynamics - Bone: Dyn_leaves_02(ComponentPose=@n8, Alpha=1, ExternalForce="0,0,0") {
    angularDampingOverride=0.999
    angularSpringConstant=400
    bAngularSpring=true
    bOverrideAngularDamping=true
    bUseGravityOverride=true
    simulationSpace=World
    alphaBoolBlend {blendOption=Linear}
    boundBone {boneName=Dyn_leaves_02}
    componentAppliedLinearAccClamp=<500 500 500>
    componentLinearAccScale=<0.2 0.2 0.2>
    componentLinearVelScale=<0.2 0.2 0.2>
    gravityOverride {z=-980}
    physicsBodyDefinitions [{
      boundBone {boneName=Dyn_leaves_02}
      boxExtents=<10 10 10>
      constraintSetup {angularLimitsMax=<0 13 13>, angularLimitsMin=<0 -30 -13>, angularTarget=<1 0 0>}
      localJointOffset=<0.1 0 0>
     }]
   }
   n10 = Anim Dynamics - Bone: Dyn_leaves_03(ComponentPose=@n9, Alpha=1, ExternalForce="0,0,0") {
    angularDampingOverride=0.999
    angularSpringConstant=400
    bAngularSpring=true
    bOverrideAngularDamping=true
    bUseGravityOverride=true
    simulationSpace=World
    alphaBoolBlend {blendOption=Linear}
    boundBone {boneName=Dyn_leaves_03}
    componentAppliedLinearAccClamp=<500 500 500>
    componentLinearAccScale=<0.2 0.2 0.2>
    componentLinearVelScale=<0.2 0.2 0.2>
    gravityOverride {z=-980}
    physicsBodyDefinitions [{
      boundBone {boneName=Dyn_leaves_03}
      boxExtents=<10 10 10>
      constraintSetup {angularLimitsMax=<0 13 13>, angularLimitsMin=<0 -13 -13>, angularTarget=<1 0 0>}
      localJointOffset=<0.1 0 0>
     }]
   }
   n11 = Anim Dynamics - Bone: Dyn_leaves_04(ComponentPose=@n10, Alpha=1, ExternalForce="0,0,0") {
    angularDampingOverride=0.999
    angularSpringConstant=400
    bAngularSpring=true
    bOverrideAngularDamping=true
    bUseGravityOverride=true
    simulationSpace=World
    alphaBoolBlend {blendOption=Linear}
    boundBone {boneName=Dyn_leaves_04}
    componentAppliedLinearAccClamp=<500 500 500>
    componentLinearAccScale=<0.2 0.2 0.2>
    componentLinearVelScale=<0.2 0.2 0.2>
    gravityOverride {z=-980}
    physicsBodyDefinitions [{
      boundBone {boneName=Dyn_leaves_04}
      boxExtents=<10 10 10>
      constraintSetup {angularLimitsMax=<0 15 15>, angularLimitsMin=<0 -30 -15>, angularTarget=<1 0 0>}
      localJointOffset=<0.1 0 0>
     }]
   }
   n12 = Anim Dynamics - Bone: Dyn_leaves_05(ComponentPose=@n11, Alpha=1, ExternalForce="0,0,0") {
    angularDampingOverride=0.999
    angularSpringConstant=400
    bAngularSpring=true
    bOverrideAngularDamping=true
    bUseGravityOverride=true
    simulationSpace=World
    alphaBoolBlend {blendOption=Linear}
    boundBone {boneName=Dyn_leaves_05}
    componentAppliedLinearAccClamp=<500 500 500>
    componentLinearAccScale=<0.2 0.2 0.2>
    componentLinearVelScale=<0.2 0.2 0.2>
    gravityOverride {z=-980}
    physicsBodyDefinitions [{
      boundBone {boneName=Dyn_leaves_05}
      boxExtents=<10 10 10>
      constraintSetup {angularLimitsMax=<0 0 0>, angularLimitsMin=<0 -30 0>, angularTarget=<1 0 0>}
      localJointOffset=<0.1 0 0>
     }]
   }
   n13 = Anim Dynamics - Bone: Dyn_leaves_06(ComponentPose=@n12, Alpha=1, ExternalForce="0,0,0") {
    angularDampingOverride=0.999
    angularSpringConstant=400
    bAngularSpring=true
    bOverrideAngularDamping=true
    bUseGravityOverride=true
    simulationSpace=World
    alphaBoolBlend {blendOption=Linear}
    boundBone {boneName=Dyn_leaves_06}
    componentAppliedLinearAccClamp=<500 500 500>
    componentLinearAccScale=<0.2 0.2 0.2>
    componentLinearVelScale=<0.2 0.2 0.2>
    gravityOverride {z=-980}
    physicsBodyDefinitions [{
      boundBone {boneName=Dyn_leaves_06}
      boxExtents=<10 10 10>
      constraintSetup {angularLimitsMax=<0 15 15>, angularLimitsMin=<0 -15 -15>, angularTarget=<1 0 0>}
      localJointOffset=<0.1 0 0>
     }]
   }
   n14 = Anim Dynamics - Bone: Dyn_leaves_07(ComponentPose=@n13, Alpha=1, ExternalForce="0,0,0") {
    angularDampingOverride=0.999
    angularSpringConstant=200
    bAngularSpring=true
    bOverrideAngularDamping=true
    bUseGravityOverride=true
    simulationSpace=World
    alphaBoolBlend {blendOption=Linear}
    boundBone {boneName=Dyn_leaves_07}
    componentAppliedLinearAccClamp=<500 500 500>
    componentLinearAccScale=<0.2 0.2 0.2>
    componentLinearVelScale=<0.2 0.2 0.2>
    gravityOverride {z=-980}
    physicsBodyDefinitions [{
      boundBone {boneName=Dyn_leaves_07}
      boxExtents=<10 10 10>
      constraintSetup {angularLimitsMax=<0 25 15>, angularLimitsMin=<0 -15 -15>, angularTarget=<1 0 0>}
      localJointOffset=<0.1 0 0>
     }]
   }
   Output Pose(Result=Blend Poses by bool(BlendPose_0=Component To Local(ComponentPose=@n14), BlendPose_1=Component To Local(ComponentPose=@n7), bActiveValue=Get bIsInFrontEnd()) {blendType=Linear}) {layerGroup=None}
   n15 = "Transform (Modify) Bone - Bone: Bodt"(ComponentPose=Local To Component(LocalPose=Input Pose()), Alpha=1, Translation="0,3.3,0", Rotation="0,0,-15", Scale="1,1,1") {
    rotationMode=BMM_Additive
    rotationSpace=BCS_BoneSpace
    scaleMode=BMM_Additive
    scaleSpace=BCS_BoneSpace
    translationMode=BMM_Additive
    translationSpace=BCS_BoneSpace
    alphaBoolBlend {blendOption=Linear}
    boneToModify {boneName=Bodt}
    rotation {roll=-15}
    translation {y=3.3}
   }
   n16 = "Transform (Modify) Bone - Bone: Bodt"!(ComponentPose=@n15, bAlphaBoolEnabled:bool=Get bIsInFrontEnd(), Translation="0,1,0", Rotation="0,0,0", Scale="1,1,1") -> (Pose) {
    alphaInputType=Bool
    rotationMode=BMM_Additive
    rotationSpace=BCS_BoneSpace
    scaleMode=BMM_Additive
    scaleSpace=BCS_BoneSpace
    translationMode=BMM_Additive
    translationSpace=BCS_BoneSpace
    alphaBoolBlend {blendOption=Linear}
    boneToModify {boneName=Bodt}
    translation {y=1}
   }
   n17 = "Transform (Modify) Bone - Bone: Bodt"(ComponentPose=@n16, bAlphaBoolEnabled=Get bIsSkyDiving(), Translation="0,-2,0", Rotation="0,0,-5", Scale="1,1,1") {
    alphaInputType=Bool
    rotationMode=BMM_Additive
    rotationSpace=BCS_BoneSpace
    scaleMode=BMM_Additive
    scaleSpace=BCS_BoneSpace
    translationMode=BMM_Additive
    translationSpace=BCS_BoneSpace
    alphaBoolBlend {blendOption=Linear}
    boneToModify {boneName=Bodt}
    rotation {roll=-5}
    translation {y=-2}
   }
   n18 = "Transform (Modify) Bone - Bone: Bodt"!(ComponentPose=@n17, Alpha:float=1, Translation="10,-8,5", Rotation="-10,40,12", Scale="1,1,1") -> (Pose) {
    rotationMode=BMM_Additive
    rotationSpace=BCS_BoneSpace
    scaleMode=BMM_Additive
    scaleSpace=BCS_BoneSpace
    translationMode=BMM_Additive
    translationSpace=BCS_BoneSpace
    alphaBoolBlend {blendOption=Linear}
    boneToModify {boneName=Bodt}
    rotation=<rot -10 40 12>
    translation=<10 -8 5>
   }
   n19 = "Transform (Modify) Bone - Bone: Dyn_leaves_05"(ComponentPose=@n18, Alpha=1, Translation="0,0,0", Rotation="1.006,14.463,-3.740", Scale="1,1,1") {
    rotationMode=BMM_Additive
    rotationSpace=BCS_BoneSpace
    scaleMode=BMM_Additive
    scaleSpace=BCS_BoneSpace
    translationMode=BMM_Additive
    translationSpace=BCS_BoneSpace
    alphaBoolBlend {blendOption=Linear}
    boneToModify {boneName=Dyn_leaves_05}
    rotation=<rot 1.006 14.463 -3.74>
   }
   Save cached pose 'Base'(Pose=Component To Local(ComponentPose=@n19)) {cacheName=Base}
  ]}
 EventGraph {nodes [
   Event BlueprintUpdateAnimation()
   n1 = GetDeltaTransformFromRefPose!(self=GetOwnerSkeletalMeshComponent(), BoneName=spine_05, BaseName=spine_01) -> (ReturnValue_Location:Vector, ReturnValue_Rotation:Rotator, ReturnValue_Scale:Vector)
   > Set Parashuting(Parashuting:double=MapRangeClamped(Value=BreakRotator(InRot=@n1.ReturnValue_Rotation).Yaw, InRangeA=-15, InRangeB=-50)) -> (Output_Get:double)
   TryGetPawnOwner()
  ]}
]
variables [
 Parashuting {type=number}
 SimlationSpace {type=AnimPhysSimSpaceType}
]
